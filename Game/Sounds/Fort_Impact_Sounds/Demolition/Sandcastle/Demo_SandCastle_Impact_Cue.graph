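# Sound cue for impact hits on sandcastles in the demolition impact sound set.
# A random node picks among four equally weighted impact waves, and a modulator varies pitch on each play.
# It uses small-object demolition hit attenuation and the impacts sound class, with a concurrency set that limits simultaneous object impacts.
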SoundCue /Game/Sounds/Fort_Impact_Sounds/Demolition/Sandcastle/Demo_SandCastle_Impact_Cue
root=Modulator_5
nodes [
 {id=Modulator_5, children [Random_4], properties {pitchMax=1.3}}
 {id=Random_4, children [WavePlayer_5, WavePlayer_12, WavePlayer_13, WavePlayer_14], properties {weights [1,1,1,1]}}
 {id=WavePlayer_12, properties {soundWaveAssetPtr=&/Demo_SandCastle_Impact_02}}
 {id=WavePlayer_13, properties {soundWaveAssetPtr=&/Demo_SandCastle_Impact_03}}
 {id=WavePlayer_14, properties {soundWaveAssetPtr=&/Demo_SandCastle_Impact_04}}
 {id=WavePlayer_5, properties {soundWaveAssetPtr=&/Demo_SandCastle_Impact_01}}
]
properties {
 attenuationSettings=/Game/Sounds/Attenuation/Fort_Demolition_Hit_Small_Sized_Object_Att
 maxDistance=4000
 soundClassObject=&Fort_Impacts
 volumeMultiplier=0.96
 concurrencyOverrides {resolutionRule=PreventNew}
 concurrencySet [&CONC_Object_Impacts_FourMax]
}
