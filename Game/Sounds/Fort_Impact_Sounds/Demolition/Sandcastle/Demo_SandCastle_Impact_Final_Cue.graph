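# Sound cue for the final demolition impact on a sandcastle, classed as an impact sound.
# A Random node picks one of four equally weighted impact waves and feeds a Modulator that varies pitch, with a raised upper limit.
# It plays at reduced volume with medium-sized-object destruction attenuation and uses a shared object-impact concurrency set named for a four-sound cap.
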SoundCue /Game/Sounds/Fort_Impact_Sounds/Demolition/Sandcastle/Demo_SandCastle_Impact_Final_Cue
root=Modulator_4
nodes [
 {id=Modulator_4, children [Random_2], properties {pitchMax=1.3}}
 {id=Random_2, children [WavePlayer_5, WavePlayer_7, WavePlayer_8, WavePlayer_10], properties {weights [1,1,1,1]}}
 {id=WavePlayer_10, properties {soundWaveAssetPtr=&/Demo_SandCastle_Impact_04}}
 {id=WavePlayer_5, properties {soundWaveAssetPtr=&/Demo_SandCastle_Impact_01}}
 {id=WavePlayer_7, properties {soundWaveAssetPtr=&/Demo_SandCastle_Impact_02}}
 {id=WavePlayer_8, properties {soundWaveAssetPtr=&/Demo_SandCastle_Impact_03}}
]
properties {
 attenuationSettings=/Game/Sounds/Attenuation/Fort_Demolition_Destroy_Medium_Sized_Object_Att
 maxDistance=4000
 soundClassObject=&Fort_Impacts_PSM
 volumeMultiplier=0.34
 concurrencyOverrides {resolutionRule=PreventNew}
 concurrencySet [&CONC_Object_Impacts_FourMax]
}
